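# NS_Env_XLCity_LightPost_Sparks is a Niagara environment effect for sparks on an XLCity light post.
# It holds three emitters, named SparksBlueWire, SmokeCameraFacing and Empty, which by name pair blue wire sparks with camera-facing smoke.
# The system lists only the SystemState module and exposes no user variables.
# No renderers, dynamic inputs or data interfaces are recorded for it.
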NiagaraSystem /Game/Environments/Hera/FX/XLCity/NS_Env_XLCity_LightPost_Sparks
dependencies {usedDataInterfaces [], usedDynamicInputs [], usedModules [&SystemState]}
emitters {
 Empty {rendererClasses [], renderers []}
 SmokeCameraFacing {rendererClasses [], renderers []}
 SparksBlueWire {rendererClasses [], renderers []}
}
summary {
 systemName=NS_Env_XLCity_LightPost_Sparks
 emitters [
  {emitterName=SparksBlueWire, rendererClasses []}
  {emitterName=SmokeCameraFacing, rendererClasses []}
  {emitterName=Empty, rendererClasses []}
 ]
 userVariables []
}
